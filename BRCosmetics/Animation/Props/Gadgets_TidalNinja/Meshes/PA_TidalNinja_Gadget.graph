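# Physics asset for the Tidal Ninja gadget prop, defining collision bodies and joint limits for its skeleton.
# A large capsule covers the root bone, while main and six JNT_Sphere bones each get a small, near-spherical capsule.
# Constraints form a star, linking main to root and every sphere joint to main with identical swing and twist limit angles.
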PhysicsAsset /BRCosmetics/Animation/Props/Gadgets_TidalNinja/Meshes/PA_TidalNinja_Gadget
bodies {
 JNT_Sphere_01 [{shapeType=Capsule, length=0.505, radius=6.1896, rotation=<rot 0 -90 0>}]
 JNT_Sphere_02 [{shapeType=Capsule, length=0.505, radius=6.1896, rotation=<rot 0 -90 0>}]
 JNT_Sphere_03 [{shapeType=Capsule, length=0.505, radius=6.1896, rotation=<rot 0 -90 0>}]
 JNT_Sphere_04 [{shapeType=Capsule, length=0.505, radius=6.1896, rotation=<rot 0 -90 0>}]
 JNT_Sphere_05 [{shapeType=Capsule, length=0.505, radius=6.1896, rotation=<rot 0 -90 0>}]
 JNT_Sphere_06 [{shapeType=Capsule, length=0.505, radius=6.1896, rotation=<rot 0 -90 0>}]
 main [{shapeType=Capsule, length=0.505, radius=6.1896, rotation=<rot 0 -90 0>}]
 root [{shapeType=Capsule, length=1.0983, radius=73.2412, rotation=<rot 90 -90 0>}]
}
constraints [
 {bone1Name=main, bone2Name=root, swing1LimitDegrees=45, swing2LimitDegrees=45, twistLimitDegrees=45}
 {bone1Name=JNT_Sphere_01, bone2Name=main, swing1LimitDegrees=45, swing2LimitDegrees=45, twistLimitDegrees=45}
 {bone1Name=JNT_Sphere_02, bone2Name=main, swing1LimitDegrees=45, swing2LimitDegrees=45, twistLimitDegrees=45}
 {bone1Name=JNT_Sphere_03, bone2Name=main, swing1LimitDegrees=45, swing2LimitDegrees=45, twistLimitDegrees=45}
 {bone1Name=JNT_Sphere_04, bone2Name=main, swing1LimitDegrees=45, swing2LimitDegrees=45, twistLimitDegrees=45}
 {bone1Name=JNT_Sphere_05, bone2Name=main, swing1LimitDegrees=45, swing2LimitDegrees=45, twistLimitDegrees=45}
 {bone1Name=JNT_Sphere_06, bone2Name=main, swing1LimitDegrees=45, swing2LimitDegrees=45, twistLimitDegrees=45}
]
